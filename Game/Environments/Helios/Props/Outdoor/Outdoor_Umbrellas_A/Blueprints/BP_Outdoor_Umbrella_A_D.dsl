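; Outdoor umbrella prop that applies a bounce gameplay effect to actors landing in or entering its Jump_Collision component.
; On overlap it gets the other actor's Ability System Component and, if valid, applies the BounceGE class to it.
; It then sends an Athena.Quests.Jump.OnUmbrella stat event to the quest manager of that actor's Fortnite player controller.
; BeginPlay, Tick, actor overlap and the construction script are empty.
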(event EventBeginPlay)

(event Collision|EventActorBeginOverlap
  (params (OtherActor :type Actor%20Object%20Reference)))

(event EventTick
  (params (DeltaSeconds :type Float%20%28single-precision%29)))

(event OnComponentBeginOverlap(Jump_Collision)
  (params (OverlappedComponent :type Primitive%20Component%20Object%20Reference) (OtherActor :type Actor%20Object%20Reference) (OtherComp :type Primitive%20Component%20Object%20Reference) (OtherBodyIndex :type Integer) (bFromSweep :type Boolean) (SweepResult :type Hit%20Result%20Structure%20%28by%20ref%29))
  (bind _returnvalue (Ability|GetAbilitySystemComponent OtherActor))
  (Utilities|IsValid _returnvalue
    (:Is%20Valid
      (GameplayEffects|ApplyGameplayEffectToSelf :self _returnvalue :GameplayEffectClass (Variables|Default|GetBounceGE) :Level 0.0)
      (Quest|SendComplexCustomStatEvent (Quests|GetQuestManager (Fortnite|GetFortPlayerControllerfromActor OtherActor) "Invalid") 0 "(GameplayTags=((TagName=\"Athena.Quests.Jump.OnUmbrella\")))" "(GameplayTags=)" 1))))

(fn ConstructionScript)
